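# Material function that builds the surface attributes for Spartan vehicle body cosmetic attachments, defaulting to leather attachment skin textures.
# It samples base color, packed PBR and normal texture parameters, then outputs them as one material attributes set.
# The PBR green channel multiplies base color, blue drives metallic through a default lerp, and alpha plus a small constant offset is saturated into roughness.
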MaterialFunction /VehicleCosmetics/Bodies/Spartan/Materials/Attachments/MF_Spartan_Attachments
graph {nodes [
  n1 = TextureSampleParameter2D() {
   parameterName=PBR_Texture
   samplerType=SAMPLERTYPE_LinearColor
   texture=/VehicleCosmetics/Bodies/Spartan/Textures/Attachment/T_Spartan_AttachmentTexture
  }
  n2 = Constant() {r=0.1}
  n3 = TextureSampleParameter2D() {parameterName=Normal, samplerType=SAMPLERTYPE_Normal, texture=/VehicleCosmetics/Bodies/Spartan/Textures/Attachment/T_Spartan_Attachment_N}
  n4 = SetMaterialAttributes!(
   MaterialAttributes
   BaseColor=Multiply(
    A=TextureSampleParameter2D() {parameterName=BaseColor, texture=/VehicleCosmetics/Bodies/Spartan/Skins/Spartan_A/T_Skin_SpartanA_LeatherAttachment_BC}.RGB
    B=@n1.G
   )
   Metallic=LinearInterpolate(Alpha=@n1.B)
   Specular
   Roughness=Saturate(None=Add(A=@n1.A, B=@n2))
   Normal=@n3.RGB
  ) {attributeSetTypes [
    #1
    #2
    #3
    #4
    #5
   ]}
  FunctionOutput(None=@n4)
 ]}
